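# Prototype Default Lit material for dark volcanic rock with glowing lava and heat, under Figment S08 Volcano.
# It blends a world-aligned cliff base with an upward-facing detail top layer built from height lerps, with vertex color driving variation and emissive cracks.
# World-height gradients, lava ground noise, masks and Fresnel shape the added glow.
# Quality switches drop that glow on Low and Medium; ES2 and ES3_1 use a simplified untextured layer blend with masked emissive.
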Material /Figment_S08/Athena/Environments/Volcano/Prototype/M_Athena_Proto_DarkRock_03
graph {
 properties {bUseMaterialAttributes=true, pixelDepthOffsetMode=PDOM_Legacy, usedShadingModels=Default Lit}
 nodes [
  n1 = VectorParameter() {parameterName=BaseColor2, defaultValue=<rgba 0.0086 0.0086 0.01 0>}
  n2 = TextureObject() {samplerType=SAMPLERTYPE_Normal, texture=/Game/Environments/Sets/Arid/Textures/RockCliff_normal}
  n3 = ScalarParameter() {defaultValue=4103.5996, parameterName=Param_2}
  n4 = TextureCoordinate()
  n5 = Multiply(A=@n4, B=Constant() {r=3.9372})
  n6 = TextureSample(UVs=@n5) {texture=/Figment_S08/Athena/Environments/Volcano/Textures/Rocks_basecolor}
  n7 = TextureSample(UVs=@n5) {samplerType=SAMPLERTYPE_LinearGrayscale, texture=/Figment_S08/Athena/Environments/Volcano/Textures/Detail_height}
  n8 = MaterialFunctionCall(
   Transition Phase=ScalarParameter() {defaultValue=0.51, group=Top, parameterName=HMTrTop}
   Height Texture=@n7.R
   Contrast=ScalarParameter() {defaultValue=234.6794, group=Top, parameterName=HMContrastTop}
  ) {materialFunction=&HeightLerp}
  n9 = LinearInterpolate(A=Multiply(A=VectorParameter() {defaultValue=<rgba 0.0625 0.0625 0.0625 0>}.RGB, B=@n6.RGB), B=@n1.RGB, Alpha=@n8.Alpha)
  n10 = TextureSample(UVs=Multiply(A=@n4, B=Constant() {r=33.7223})) {samplerType=SAMPLERTYPE_Normal, texture=/Game/Athena/Environments/Volcano/Textures/Small_Detail_normal}
  n11 = Add(A=TextureSample(UVs=@n5) {samplerType=SAMPLERTYPE_Normal, texture=/Game/Athena/Environments/Volcano/Textures/Detail_normal}.RGB, B=@n10.RGB)
  n12 = Constant3Vector() {constant=<rgba 3 3 1 0>}
  n13 = WorldPosition()
  n14 = Constant() {r=1024.0447}
  n15 = TextureSample(UVs=ComponentMask(None=Divide(A=@n13.XYZ, B=@n14)) {g=true, r=true}) {samplerType=SAMPLERTYPE_Normal, texture=/Game/Global/Textures/Default/Noise/T_SwirlNoise_N}
  n16 = Add(
   A=MaterialFunctionCall(Normal=@n15.RGB, Flatness=Constant() {r=0.9233}) {materialFunction=&FlattenNormal}.Result
   B=Multiply(A=@n10.RGB, B=Constant3Vector() {constant=<rgba 0.5 0.5 0 0>})
  )
  n17 = Constant3Vector() {constant=<rgba 0 0 1 0>}
  n18 = Fresnel(Normal=VertexNormalWS()) {baseReflectFraction=0, exponent=3}
  n19 = Constant3Vector() {constant=<rgba 0.125 0.125 0 0>}
  n20 = Constant() {r=0.98}
  n21 = VertexColor()
  n22 = MaterialFunctionCall(
   A=LinearInterpolate(A=Multiply(A=@n11, B=@n12), B=LinearInterpolate(A=@n16, B=@n17, Alpha=@n18), Alpha=@n8.Alpha)
   B=Add(
    A=Multiply(A=@n10.RGB, B=@n19)
    B=MaterialFunctionCall(Normal=@n15.RGB, Flatness=@n20) {materialFunction=&FlattenNormal}.Result
   )
   Transition Phase=@n21.R
   Height Texture=@n7.R
   Contrast=Constant() {r=5}
  ) {materialFunction=&HeightLerp}
  n23 = LinearInterpolate(B=Constant() {r=1}, Alpha=@n8.Alpha)
  n24 = ScalarParameter() {defaultValue=0.2, group=Top, parameterName=STop2}
  n25 = LinearInterpolate(A=ScalarParameter() {defaultValue=-0.0381, group=Top, parameterName=STop}, B=@n24, Alpha=@n8.Alpha)
  n26 = ScalarParameter() {defaultValue=0.125, group=Top, parameterName=RTop2}
  n27 = LinearInterpolate(A=ScalarParameter() {defaultValue=0.8685, group=Top, parameterName=RTop}, B=@n26, Alpha=@n8.Alpha)
  n28 = Constant() {r=4}
  n29 = Constant() {r=5.4426}
  n30 = MaterialFunctionCall(In=@n8.Lerp Alpha No Contrast, Contrast=Constant() {r=9.489}) {materialFunction=&CheapContrast}
  n31 = TextureSample(UVs=BumpOffset(
    Coordinate=@n5
    Height=TextureSample() {samplerType=SAMPLERTYPE_LinearGrayscale, texture=&Detail_height}.R
    HeightRatioInput=Constant() {r=-0.1}
   ) {referencePlane=-0.5}) {samplerType=SAMPLERTYPE_LinearGrayscale, texture=&Detail_height}
  n32 = Constant() {r=1.6704}
  n33 = Multiply(
   A=Multiply(A=Multiply(A=@n6.R, B=@n28), B=LinearInterpolate(A=@n29, Alpha=@n30.Result) {constA=2})
   B=MaterialFunctionCall(Transition Phase=@n21.G, Height Texture=Multiply(A=@n31.R, B=@n32)) {materialFunction=&HeightLerp}.Alpha
  )
  n34 = Constant3Vector() {constant=<rgba 0.5573 0.0237 0 1>}
  n35 = Constant() {r=-3000.5796}
  n36 = Constant() {r=49815.8984}
  n37 = DotProduct(A=VertexNormalWS(), B=Constant3Vector() {constant=<rgba 0 0 1 0>})
  n38 = MaterialFunctionCall!(
   Base Material=MakeMaterialAttributes!(
    BaseColor=@n1.RGB
    Metallic
    Specular=Constant() {r=0.2}
    Roughness=Constant() {r=0.3}
    Anisotropy
    EmissiveColor
    Opacity
    OpacityMask
    Normal=MaterialFunctionCall(TextureObject=@n2, TextureSize=@n3) {materialFunction=&WorldAlignedNormal}.XYZ Texture
    Tangent
    WorldPositionOffset
    SubsurfaceColor
    ClearCoat
    ClearCoatRoughness
    AmbientOcclusion
    Refraction
    CustomizedUVs_0
    CustomizedUVs_1
    CustomizedUVs_2
    CustomizedUVs_3
    CustomizedUVs_4
    CustomizedUVs_5
    CustomizedUVs_6
    CustomizedUVs_7
    PixelDepthOffset
    ShadingModel
    Displacement
   )
   Top Material=MakeMaterialAttributes(
    BaseColor=LinearInterpolate(A=@n9, B=@n1.RGB, Alpha=@n22.Alpha)
    Metallic=LinearInterpolate(A=@n23, Alpha=@n22.Alpha)
    Specular=LinearInterpolate(A=@n25, B=@n24, Alpha=@n22.Alpha)
    Roughness=LinearInterpolate(A=@n27, B=@n26, Alpha=@n22.Alpha)
    EmissiveColor=Multiply(A=@n33, B=@n34)
    Normal=@n22.Results
   )
   Alpha=Saturate(None=LinearInterpolate(A=@n35, B=@n36, Alpha=@n37))
  ) {materialFunction=/Engine/Functions/MaterialLayerFunctions/MatLayerBlend_Standard}
  n39 = GetMaterialAttributes(None=@n38.Blended Material) {attributeGetTypes [#1, #2]}
  n40 = ComponentMask(None=WorldPosition().XYZ) {b=true}
  n41 = ScalarParameter() {defaultValue=8841.3438, parameterName=Param_9}
  n42 = Saturate(None=Divide(A=Subtract(A=@n40, B=@n41), B=ScalarParameter() {defaultValue=-555, parameterName=Param_8}))
  n43 = Multiply(
   A=VectorParameter() {group=Lava, parameterName=GradientColor3, defaultValue=<rgba 3 0.9406 0 1>}.RGB
   B=Power(Base=@n42, Exp=ScalarParameter() {defaultValue=28.633, group=Lava, parameterName=EmissiveEdgeContrast2})
  )
  n44 = TextureCoordinate()
  n45 = ScalarParameter() {defaultValue=3.8718, group=Lava, parameterName=LavaEdgeTiling}
  n46 = OneMinus(None=MaterialFunctionCall!(UV=Multiply(A=@n44, B=@n45), WorldAlignedSize) {materialFunction=/Game/Environments/Sets/Vulcano/Materials/MF_LavaGround}.Height)
  n47 = ScalarParameter() {defaultValue=3.1021, group=Lava, parameterName=EmissiveEdgeContrast}
  n48 = ScalarParameter() {defaultValue=0.658, group=Lava, parameterName=EmissiveEdgeHeightLerp}
  n49 = Multiply(
   A=LinearInterpolate(
    A=VectorParameter() {group=Lava, parameterName=GradientColor1, defaultValue=<rgba 0 0 0 1>}.RGB
    B=VectorParameter() {group=Lava, parameterName=GradientColor2, defaultValue=<rgba 1 0.082 0 1>}.RGB
    Alpha=Power(Base=@n46, Exp=@n47)
   )
   B=MaterialFunctionCall(
    Transition Phase=Multiply(A=@n42, B=@n48)
    Height Texture=@n46
    Contrast=ScalarParameter() {defaultValue=1.1432, group=Lava, parameterName=ContrastEdgeEmissive}
   ) {materialFunction=&HeightLerp}.Alpha
  )
  n50 = Power(
   Base=OneMinus(None=TextureSampleParameter2D() {parameterName=Mask, samplerType=SAMPLERTYPE_Masks, texture=/Figment_S08/Athena/Environments/Volcano/Textures/Volcano_E}.R)
   Exp=ScalarParameter() {defaultValue=7.0831, parameterName=Param_6}
  )
  n51 = Multiply(
   A=Power(
    Base=TextureSample() {samplerType=SAMPLERTYPE_Masks, texture=/Figment_S08/Athena/Environments/Volcano/Textures/T_Stairs_S}.R
    Exp=Constant() {r=3}
   )
   B=Constant() {r=0.8}
  )
  n52 = VectorParameter() {parameterName=Heat2, defaultValue=<rgba 0.959 0.062 0 0>}
  n53 = MaterialFunctionCall!(
   Low _ PC Low and Med=@n52.RGB
   "High _ XB1, PS4"=@n52.RGB
   "Epic _ PC High, XSX, PS5"=Constant()
   --------
   "Mobile Low _ PC Perf, Android"=@n52.RGB
   Mobile Med _ Switch=@n52.RGB
  ) {materialFunction=/Game/Global/MaterialFunctions/TechArt/MF_Fortnite_QualitySwitch_Vector3}
  n54 = Multiply(A=Multiply(A=Add(A=@n50, B=@n51), B=@n53.Result), B=Constant() {r=848.5})
  n55 = Multiply(A=Constant3Vector() {constant=<rgba 1.25 0.0717 0 1>}, B=Constant() {r=0.381})
  n56 = Subtract(A=WorldPosition().XYZ, B=ScalarParameter() {defaultValue=8552.3945, parameterName=Paramffff})
  n57 = Saturate(None=Divide(A=ComponentMask(None=@n56) {b=true}, B=ScalarParameter() {defaultValue=317.4202, parameterName=s}))
  n58 = WorldPosition()
  n59 = ScalarParameter() {defaultValue=8400.5518, parameterName=Paramffffa}
  n60 = Divide(A=ComponentMask(None=Subtract(A=@n58.XYZ, B=@n59)) {b=true}, B=ScalarParameter() {defaultValue=245.5631, parameterName=sa})
  n61 = Add(
   A=Add(A=Multiply(A=@n54, B=@n39.Base Color), B=@n39.Emissive Color)
   B=Multiply(A=LinearInterpolate(A=@n55, Alpha=@n57) {constB=0}, B=Saturate(None=@n60))
  )
  n62 = OneMinus(None=Fresnel(Normal=VertexNormalWS()) {exponent=2.5})
  n63 = SetMaterialAttributes(MaterialAttributes=@n39.MaterialAttributes, Emissive Color=Add(A=Add(A=@n43, B=@n49), B=Multiply(A=@n61, B=@n62))) {attributeSetTypes [#2]}
  n64 = QualitySwitch(Default=@n63, Low=@n38.Blended Material, High=@n63, Medium=@n38.Blended Material)
  n65 = VectorParameter() {parameterName=BaseColor2, defaultValue=<rgba 0.0086 0.0086 0.01 0>}
  n66 = Constant() {r=0.2}
  n67 = Constant() {r=0.3}
  n68 = Constant() {r=1}
  n69 = Constant() {r=0.125}
  n70 = SetMaterialAttributes(
   MaterialAttributes=MaterialFunctionCall(
    Base Material=MakeMaterialAttributes(BaseColor=@n65.RGB, Specular=@n66, Roughness=@n67)
    Top Material=MakeMaterialAttributes(BaseColor=@n65.RGB, Metallic=@n68, Specular=@n69, Roughness=@n69)
    Alpha=Saturate(None=@n37)
   ) {materialFunction=&MatLayerBlend_Standard}.Blended Material
   Emissive Color=Multiply(A=@n50, B=@n53.Result)
  ) {attributeSetTypes [#2]}
  FeatureLevelSwitch(Default=@n64, ES2=@n70, ES3_1=@n70, SM4_REMOVED=@n64, SM5=@n64, SM6=@n64)
 ]
}
