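# Prototype volcano material for rocks sitting in lava: darkened rock textures with emissive heat driven by absolute world-height gradients.
# Interior heat color multiplies the diffuse (sampled at fixed mip 2 except on ES2) and height masks, but defaults to black.
# A lava edge glow grows at lower heights, blending color gradients through MF_LavaGround height and HeightLerp.
# Low and Medium quality, ES2 and ES3_1 drop the edge glow.
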Material /Figment_S08/Athena/Environments/Volcano/Prototype/M_RocksInLava
graph {
 properties {bUseMaterialAttributes=true, pixelDepthOffsetMode=PDOM_Legacy, usedShadingModels=Default Lit}
 nodes [
  Constant3Vector() {constant=<rgba 1 0.082 0 1>}
  Constant3Vector() {constant=<rgba 3 0.9406 0 1>}
  Constant3Vector() {constant=<rgba 0.959 0.062 0 0>}
  n1 = TextureSampleParameter2D!(UVs, Level=Constant() {r=2}, Apply View MipBias) {mipValueMode=TMVM_MipLevel, parameterName=D, texture=/Game/Environments/Props/Rocks/Textures/T_Lava_Rock_DE_D}
  n2 = FeatureLevelSwitch(
   Default=@n1.RGB
   ES2=TextureSampleParameter2D!(UVs, Apply View MipBias) {parameterName=D, texture=&T_Lava_Rock_DE_D}.RGB
   ES3_1=@n1.RGB
   SM4_REMOVED=@n1.RGB
   SM5=@n1.RGB
   SM6=@n1.RGB
  )
  n3 = TextureSampleParameter2D() {parameterName=S, samplerType=SAMPLERTYPE_Masks, texture=/Game/Environments/Props/Rocks/Textures/T_Lava_Rock_DE_S}
  n4 = VectorParameter() {group=Interior, parameterName=HeatInterior, defaultValue=<rgba 0 0 0 1>}
  n5 = WorldPosition()
  n6 = Subtract(A=@n5.XYZ, B=Constant() {r=13823.8096})
  n7 = Saturate(None=Divide(A=ComponentMask(None=@n6) {b=true}, B=Constant() {r=-5885.4053}))
  n8 = ScalarParameter() {defaultValue=7, parameterName=GlowInteriorLine}
  n9 = Multiply(A=Multiply(A=@n4.RGB, B=Saturate(None=@n7)), B=@n8)
  n10 = Multiply(A=@n4.RGB, B=@n8)
  n11 = Divide(A=Subtract(A=ComponentMask(None=@n5.XYZ) {b=true}, B=Constant() {r=8448.2754}), B=Constant() {r=389.197})
  n12 = Multiply(A=@n2, B=Multiply(A=FeatureLevelSwitch(Default=@n9, ES2=@n10, ES3_1=@n9, SM4_REMOVED=@n9, SM5=@n9, SM6=@n9), B=Saturate(None=@n11)))
  n13 = VectorParameter() {group=Lava, parameterName=GradientColor3, defaultValue=<rgba 3 0.9406 0 1>}
  n14 = ComponentMask(None=WorldPosition().XYZ) {b=true}
  n15 = ScalarParameter() {defaultValue=9404.5156}
  n16 = Saturate(None=Divide(A=Subtract(A=@n14, B=@n15), B=ScalarParameter() {defaultValue=-1271.9333, parameterName=Param_1}))
  n17 = Power(Base=@n16, Exp=ScalarParameter() {defaultValue=28.633, group=Lava, parameterName=EmissiveEdgeContrast2})
  n18 = TextureCoordinate()
  n19 = ScalarParameter() {defaultValue=3.8718, group=Lava, parameterName=LavaEdgeTiling}
  n20 = OneMinus(None=MaterialFunctionCall!(UV=Multiply(A=@n18, B=@n19), WorldAlignedSize) {materialFunction=/Game/Environments/Sets/Vulcano/Materials/MF_LavaGround}.Height)
  n21 = LinearInterpolate(
   A=VectorParameter() {group=Lava, parameterName=GradientColor1, defaultValue=<rgba 0 0 0 1>}.RGB
   B=VectorParameter() {group=Lava, parameterName=GradientColor2, defaultValue=<rgba 1 0.082 0 1>}.RGB
   Alpha=Power(Base=@n20, Exp=ScalarParameter() {defaultValue=3.1021, group=Lava, parameterName=EmissiveEdgeContrast})
  )
  n22 = MaterialFunctionCall(
   Transition Phase=Multiply(A=@n16, B=ScalarParameter() {defaultValue=0.658, group=Lava, parameterName=EmissiveEdgeHeightLerp})
   Height Texture=@n20
   Contrast=ScalarParameter() {defaultValue=1.1432, group=Lava, parameterName=ContrastEdgeEmissive}
  ) {materialFunction=&HeightLerp}
  n23 = Add(A=@n12, B=Add(A=Multiply(A=@n13.RGB, B=@n17), B=Multiply(A=@n21, B=@n22.Alpha))) {constB=0}
  n24 = QualitySwitch(Default=@n23, Low=@n12, High=@n23, Medium=@n12)
  MakeMaterialAttributes!(
   BaseColor=Multiply(A=VectorParameter() {defaultValue=<rgba 0.0521 0.0521 0.0521 0>}.RGB, B=@n2)
   Metallic
   Specular=@n3.R
   Roughness=@n3.R
   Anisotropy
   EmissiveColor=FeatureLevelSwitch(Default=@n24, ES2=@n12, ES3_1=@n12, SM4_REMOVED=@n24, SM5=@n24, SM6=@n24)
   Opacity
   OpacityMask
   Normal=TextureSampleParameter2D() {parameterName=N, samplerType=SAMPLERTYPE_Normal, texture=/Game/Environments/Props/Rocks/Textures/T_Lava_Rock_DE_N}.RGB
   Tangent
   WorldPositionOffset
   SubsurfaceColor
   ClearCoat
   ClearCoatRoughness
   AmbientOcclusion
   Refraction
   CustomizedUVs_0
   CustomizedUVs_1
   CustomizedUVs_2
   CustomizedUVs_3
   CustomizedUVs_4
   CustomizedUVs_5
   CustomizedUVs_6
   CustomizedUVs_7
   PixelDepthOffset
   ShadingModel
   Displacement
  )
 ]
}
